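# Post-process material in the 3DHB TextureBaker mesh-capture tool that outputs the scene's diffuse color.
# A shading path switch uses the DiffuseColor scene texture on the default path.
# The Forward and Mobile paths instead show the T_3DHB_ES31 texture as a substitute image.
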Material /Game/Tools/3DHB/TextureBaker/MeshCaputer/PostShader/M_3DHB_PostDiffuse
graph {
 properties {materialDomain=MD_PostProcess, pixelDepthOffsetMode=PDOM_Legacy, usedShadingModels=Default Lit}
 nodes [
  n1 = TextureSample() {texture=/Game/Tools/3DHB/DefaultAssets/T_3DHB_ES31}
  ShadingPathSwitch(Default=SceneTexture() {sceneTextureId=PPI_DiffuseColor}.Color, Forward=@n1.RGB, Mobile=@n1.RGB)
 ]
}
