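# Post-process material in the 3DHB TextureBaker mesh capture tools that outputs scene normals as a view-space normal map color.
# It transforms the WorldNormal scene texture from world to view space, remaps it with ConstantBiasScale, then inverts green and blue.
# A ShadingPathSwitch uses this result on the default path and substitutes the T_3DHB_ES31 placeholder texture on Forward and Mobile paths.
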Material /Game/Tools/3DHB/TextureBaker/MeshCaputer/PostShader/M_3DHB_PostNormal
graph {
 properties {materialDomain=MD_PostProcess, pixelDepthOffsetMode=PDOM_Legacy, usedShadingModels=Default Lit}
 nodes [
  n1 = SceneTexture() {sceneTextureId=PPI_WorldNormal}
  n2 = MaterialFunctionCall(Float3=ConstantBiasScale(None=Transform(None=@n1.Color) {transformSourceType=TRANSFORMSOURCE_World, transformType=TRANSFORM_View})) {materialFunction=&BreakOutFloat3Components}
  n3 = TextureSample() {texture=/Game/Tools/3DHB/DefaultAssets/T_3DHB_ES31}
  ShadingPathSwitch(
   Default=MaterialFunctionCall(X=@n2.R, Y=OneMinus(None=@n2.G), Z=OneMinus(None=@n2.B)) {materialFunction=&MakeFloat3}.Result
   Forward=@n3.RGB
   Mobile=@n3.RGB
  )
 ]
}
